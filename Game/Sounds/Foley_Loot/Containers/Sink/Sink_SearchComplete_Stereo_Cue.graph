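# Sound cue for the stereo foley played when searching a sink loot container completes, reusing toilet search recordings.
# A random node picks evenly among three toilet search waves, feeding an enveloper that shapes them into a short burst with a quick attack and decay.
# A modulator at the root adds slight pitch and volume variation.
# It uses prop-search attenuation, the loot-collect sound class and a concurrency override that prevents new instances.
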SoundCue /Game/Sounds/Foley_Loot/Containers/Sink/Sink_SearchComplete_Stereo_Cue
root=Modulator_15
nodes [
 {
  id=Enveloper_8
  children [Random_2]
  properties {
   pitchCurve {editorCurveData {keys [{time=-0.0037}, <at 0.0001 0.9999>]}}
   volumeCurve {editorCurveData {keys [<at -0.001 0.0028>,<at 0.0377 1>,<at 0.25 0.4>,<at 0.6 0.0028>]}}
  }
 }
 {id=Modulator_15, children [Enveloper_8], properties {pitchMax=1.07, pitchMin=0.92, volumeMax=1.07, volumeMin=0.92}}
 {id=Random_2, children [WavePlayer_6,WavePlayer_7,WavePlayer_8], properties {weights [1,1,1]}}
 {id=WavePlayer_6, properties {soundWaveAssetPtr=/Game/Sounds/Foley_Loot/Containers/Toilet/Toilet_Search08}}
 {id=WavePlayer_7, properties {soundWaveAssetPtr=/Game/Sounds/Foley_Loot/Containers/Toilet/Toilet_Search02}}
 {id=WavePlayer_8, properties {soundWaveAssetPtr=/Game/Sounds/Foley_Loot/Containers/Toilet/Toilet_Search06}}
]
properties {
 attenuationSettings=/Game/Sounds/Attenuation/Fort_Prop_Search_Att
 bOverrideConcurrency=true
 maxDistance=1300
 soundClassObject=&Fort_Loot_Collect
 volumeMultiplier=0.6
 attenuationOverrides {falloffDistance=700, attenuationShapeExtents {x=500}}
 concurrencyOverrides {resolutionRule=PreventNew}
}
