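# Sound cue for searching a sink loot container, routed to the loot collect sound class with prop search attenuation.
# A Mixer layers a random pick of three toilet search waves with a sink wave whose pitch varies slightly through a Modulator.
# All four waves come from the Toilet container folder.
# Its concurrency override prevents new instances from starting once the limit is reached.
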SoundCue /Game/Sounds/Foley_Loot/Containers/Sink/Sink_Search_Stereo_Cue
root=Mixer_1
nodes [
 {id=Mixer_1, children [Random_2,Modulator_0], properties {inputVolume [1,1]}}
 {id=Modulator_0, children [WavePlayer_5], properties {pitchMax=1.1, pitchMin=0.9}}
 {id=Random_2, children [WavePlayer_6,WavePlayer_7,WavePlayer_8], properties {weights [1,1,1]}}
 {id=WavePlayer_5, properties {soundWaveAssetPtr=/Game/Sounds/Foley_Loot/Containers/Toilet/Sink1_Stereo}}
 {id=WavePlayer_6, properties {soundWaveAssetPtr=/Game/Sounds/Foley_Loot/Containers/Toilet/Toilet_Search08}}
 {id=WavePlayer_7, properties {soundWaveAssetPtr=/Game/Sounds/Foley_Loot/Containers/Toilet/Toilet_Search02}}
 {id=WavePlayer_8, properties {soundWaveAssetPtr=/Game/Sounds/Foley_Loot/Containers/Toilet/Toilet_Search06}}
]
properties {
 attenuationSettings=/Game/Sounds/Attenuation/Fort_Prop_Search_Att
 bOverrideConcurrency=true
 maxDistance=1300
 soundClassObject=&Fort_Loot_Collect
 volumeMultiplier=0.7
 attenuationOverrides {falloffDistance=700, attenuationShapeExtents {x=500}}
 concurrencyOverrides {resolutionRule=PreventNew}
}
